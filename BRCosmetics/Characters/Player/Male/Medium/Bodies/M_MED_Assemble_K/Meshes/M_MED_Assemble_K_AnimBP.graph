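# Animation Blueprint for the M_MED_Assemble_K player body cosmetic that poses its extra arm, shoulder plate and finger bones.
# It copies the parent mesh pose, then applies constraints, with K pinky bones following ring fingers.
# Upperarm-versus-clavicle deltas become alphas for additive shoulder plate offsets; back and down alphas go unused.
# A material parameter hides element 09 for replaced hands, except during emotes, tactical sprinting, driving or skydiving.
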AnimBlueprint /BRCosmetics/Characters/Player/Male/Medium/Bodies/M_MED_Assemble_K/Meshes/M_MED_Assemble_K_AnimBP
target_skeleton=&/M_MED_Assemble_K_Skeleton
graphs [
 AnimGraph {nodes [
   n1 = "Constraint - dyn_arm_l (R)"(ComponentPose=Local To Component(LocalPose=Copy Pose From Mesh() {bUseAttachedParent=true}), Alpha=1, ConstraintWeights_0=1) {
    alphaBoolBlend {blendOption=Linear}
    boneToModify {boneName=dyn_arm_l}
    constraintSetup [{transformType=Rotation, targetBone {boneName=upperarm_l}}]
    constraintWeights [1]
   }
   n2 = "Constraint - dyn_arm_r (R)"(ComponentPose=@n1, Alpha=1, ConstraintWeights_0=1) {
    alphaBoolBlend {blendOption=Linear}
    boneToModify {boneName=dyn_arm_r}
    constraintSetup [{transformType=Rotation, targetBone {boneName=upperarm_r}}]
    constraintWeights [1]
   }
   n3 = "Constraint - dyn_arm_aim_l (PP)"!(ComponentPose=@n2, Alpha=1, ConstraintWeights_0=0.3, ConstraintWeights_1:float=0.7) -> (Pose) {
    alphaBoolBlend {blendOption=Linear}
    boneToModify {boneName=dyn_arm_aim_l}
    constraintSetup [{targetBone {boneName=upperarm_l}}, {targetBone {boneName=dyn_arm_l}}]
    constraintWeights [1,1]
   }
   n4 = "Constraint - dyn_arm_aim_r (PP)"!(ComponentPose=@n3, Alpha=1, ConstraintWeights_0=0.3, ConstraintWeights_1:float=0.7) -> (Pose) {
    alphaBoolBlend {blendOption=Linear}
    boneToModify {boneName=dyn_arm_aim_r}
    constraintSetup [{targetBone {boneName=upperarm_r}}, {targetBone {boneName=dyn_arm_r}}]
    constraintWeights [1,1]
   }
   n5 = "Constraint - dyn_lowerarm_base_l (PP)"!(ComponentPose=@n4, Alpha=1, ConstraintWeights_0=0.7, ConstraintWeights_1:float=0.3) -> (Pose) {
    alphaBoolBlend {blendOption=Linear}
    boneToModify {boneName=dyn_lowerarm_base_l}
    constraintSetup [{targetBone {boneName=lowerarm_l}}, {targetBone {boneName=dyn_arm_l}}]
    constraintWeights [1,1]
   }
   n6 = "Constraint - dyn_lowerarm_base_r (PP)"!(ComponentPose=@n5, Alpha=1, ConstraintWeights_0=0.7, ConstraintWeights_1:float=0.3) -> (Pose) {
    alphaBoolBlend {blendOption=Linear}
    boneToModify {boneName=dyn_lowerarm_base_r}
    constraintSetup [{targetBone {boneName=lowerarm_r}}, {targetBone {boneName=dyn_arm_r}}]
    constraintWeights [1,1]
   }
   n7 = "Constraint - dyn_shoulderPlate_base_l (R)"(ComponentPose=@n6, Alpha=1, ConstraintWeights_0=1) {
    alphaBoolBlend {blendOption=Linear}
    boneToModify {boneName=dyn_shoulderPlate_base_l}
    constraintSetup [{transformType=Rotation, targetBone {boneName=clavicle_l}}]
    constraintWeights [1]
   }
   n8 = Get Shoulder_rot_up_l() -> (Shoulder_rot_up_l:double)
   n9 = "Transform (Modify) Bone - Bone: dyn_shoulderPlate_base_l"(ComponentPose=@n7, Alpha=@n8, Translation="2,-2.5,-1", Rotation="10,-5,-15", Scale="1,1,1") {
    rotationMode=BMM_Additive
    rotationSpace=BCS_BoneSpace
    scaleMode=BMM_Additive
    scaleSpace=BCS_BoneSpace
    translationMode=BMM_Additive
    translationSpace=BCS_BoneSpace
    alphaBoolBlend {blendOption=Linear}
    boneToModify {boneName=dyn_shoulderPlate_base_l}
    rotation=<rot 10 -5 -15>
    translation=<2 -2.5 -1>
   }
   n10 = "Transform (Modify) Bone - Bone: dyn_shoulderPlate_l"(ComponentPose=@n9, Alpha=@n8, Translation="0,0,0", Rotation="20,0,0", Scale="1,1,1") {
    rotationMode=BMM_Additive
    rotationSpace=BCS_BoneSpace
    scaleMode=BMM_Additive
    scaleSpace=BCS_BoneSpace
    translationMode=BMM_Additive
    translationSpace=BCS_BoneSpace
    alphaBoolBlend {blendOption=Linear}
    boneToModify {boneName=dyn_shoulderPlate_l}
    rotation {pitch=20}
   }
   n11 = "Transform (Modify) Bone - Bone: dyn_shoulderPlate_base_l"(ComponentPose=@n10, Alpha=Get Shoulder_rot_f_l() -> (Shoulder_rot_f_l:double), Translation="0,0,0", Rotation="0,0,-60", Scale="1,1,1") {
    rotationMode=BMM_Additive
    rotationSpace=BCS_BoneSpace
    scaleMode=BMM_Additive
    scaleSpace=BCS_BoneSpace
    translationMode=BMM_Additive
    translationSpace=BCS_BoneSpace
    alphaBoolBlend {blendOption=Linear}
    boneToModify {boneName=dyn_shoulderPlate_base_l}
    rotation {roll=-60}
   }
   n12 = "Constraint - dyn_shoulderPlate_base_r (R)"(ComponentPose=@n11, Alpha=1, ConstraintWeights_0=1) {
    alphaBoolBlend {blendOption=Linear}
    boneToModify {boneName=dyn_shoulderPlate_base_r}
    constraintSetup [{transformType=Rotation, targetBone {boneName=clavicle_r}}]
    constraintWeights [1]
   }
   n13 = Get Shoulder_rot_up_r() -> (Shoulder_rot_up_r:double)
   n14 = "Transform (Modify) Bone - Bone: dyn_shoulderPlate_base_r"(ComponentPose=@n12, Alpha=@n13, Translation="2,2.5,-1", Rotation="10,5,15", Scale="1,1,1") {
    rotationMode=BMM_Additive
    rotationSpace=BCS_BoneSpace
    scaleMode=BMM_Additive
    scaleSpace=BCS_BoneSpace
    translationMode=BMM_Additive
    translationSpace=BCS_BoneSpace
    alphaBoolBlend {blendOption=Linear}
    boneToModify {boneName=dyn_shoulderPlate_base_r}
    rotation=<rot 10 5 15>
    translation=<2 2.5 -1>
   }
   n15 = "Transform (Modify) Bone - Bone: dyn_shoulderPlate_r"(ComponentPose=@n14, Alpha=@n13, Translation="0,0,0", Rotation="20,0,0", Scale="1,1,1") {
    rotationMode=BMM_Additive
    rotationSpace=BCS_BoneSpace
    scaleMode=BMM_Additive
    scaleSpace=BCS_BoneSpace
    translationMode=BMM_Additive
    translationSpace=BCS_BoneSpace
    alphaBoolBlend {blendOption=Linear}
    boneToModify {boneName=dyn_shoulderPlate_r}
    rotation {pitch=20}
   }
   n16 = "Transform (Modify) Bone - Bone: dyn_shoulderPlate_base_r"(ComponentPose=@n15, Alpha=Get Shoulder_rot_f_r() -> (Shoulder_rot_f_r:double), Translation="0,1.5,0", Rotation="0,0,60", Scale="1,1,1") {
    rotationMode=BMM_Additive
    rotationSpace=BCS_BoneSpace
    scaleMode=BMM_Additive
    scaleSpace=BCS_BoneSpace
    translationMode=BMM_Additive
    translationSpace=BCS_BoneSpace
    alphaBoolBlend {blendOption=Linear}
    boneToModify {boneName=dyn_shoulderPlate_base_r}
    rotation {roll=60}
    translation {y=1.5}
   }
   n17 = "Constraint - K_thumb_02_l (R)"(ComponentPose=@n16, Alpha=1, ConstraintWeights_0=1) {
    alphaBoolBlend {blendOption=Linear}
    boneToModify {boneName=K_thumb_02_l}
    constraintSetup [{transformType=Rotation, targetBone {boneName=thumb_02_l}}]
    constraintWeights [1]
   }
   n18 = "Constraint - K_thumb_03_l (R)"(ComponentPose=@n17, Alpha=1, ConstraintWeights_0=1) {
    alphaBoolBlend {blendOption=Linear}
    boneToModify {boneName=K_thumb_03_l}
    constraintSetup [{transformType=Rotation, targetBone {boneName=thumb_03_l}}]
    constraintWeights [1]
   }
   n19 = "Constraint - K_index_02_l (R)"(ComponentPose=@n18, Alpha=1, ConstraintWeights_0=1) {
    alphaBoolBlend {blendOption=Linear}
    boneToModify {boneName=K_index_02_l}
    constraintSetup [{transformType=Rotation, targetBone {boneName=index_02_l}}]
    constraintWeights [1]
   }
   n20 = "Constraint - K_index_03_l (R)"(ComponentPose=@n19, Alpha=1, ConstraintWeights_0=1) {
    alphaBoolBlend {blendOption=Linear}
    boneToModify {boneName=K_index_03_l}
    constraintSetup [{transformType=Rotation, targetBone {boneName=index_03_l}}]
    constraintWeights [1]
   }
   n21 = "Constraint - K_pinky_02_l (R)"(ComponentPose=@n20, Alpha=1, ConstraintWeights_0=1) {
    alphaBoolBlend {blendOption=Linear}
    boneToModify {boneName=K_pinky_02_l}
    constraintSetup [{transformType=Rotation, targetBone {boneName=ring_02_l}}]
    constraintWeights [1]
   }
   n22 = "Constraint - K_pinky_03_l (R)"(ComponentPose=@n21, Alpha=1, ConstraintWeights_0=1) {
    alphaBoolBlend {blendOption=Linear}
    boneToModify {boneName=K_pinky_03_l}
    constraintSetup [{transformType=Rotation, targetBone {boneName=ring_03_l}}]
    constraintWeights [1]
   }
   n23 = "Constraint - K_thumb_02_r (R)"(ComponentPose=@n22, Alpha=1, ConstraintWeights_0=1) {
    alphaBoolBlend {blendOption=Linear}
    boneToModify {boneName=K_thumb_02_r}
    constraintSetup [{transformType=Rotation, targetBone {boneName=thumb_02_r}}]
    constraintWeights [1]
   }
   n24 = "Constraint - K_thumb_03_r (R)"(ComponentPose=@n23, Alpha=1, ConstraintWeights_0=1) {
    alphaBoolBlend {blendOption=Linear}
    boneToModify {boneName=K_thumb_03_r}
    constraintSetup [{transformType=Rotation, targetBone {boneName=thumb_03_r}}]
    constraintWeights [1]
   }
   n25 = "Constraint - K_index_02_r (R)"(ComponentPose=@n24, Alpha=1, ConstraintWeights_0=1) {
    alphaBoolBlend {blendOption=Linear}
    boneToModify {boneName=K_index_02_r}
    constraintSetup [{transformType=Rotation, targetBone {boneName=index_02_r}}]
    constraintWeights [1]
   }
   n26 = "Constraint - K_index_03_r (R)"(ComponentPose=@n25, Alpha=1, ConstraintWeights_0=1) {
    alphaBoolBlend {blendOption=Linear}
    boneToModify {boneName=K_index_03_r}
    constraintSetup [{transformType=Rotation, targetBone {boneName=index_03_r}}]
    constraintWeights [1]
   }
   n27 = "Constraint - K_pinky_02_r (R)"(ComponentPose=@n26, Alpha=1, ConstraintWeights_0=1) {
    alphaBoolBlend {blendOption=Linear}
    boneToModify {boneName=K_pinky_02_r}
    constraintSetup [{transformType=Rotation, targetBone {boneName=ring_02_r}}]
    constraintWeights [1]
   }
   n28 = "Constraint - K_pinky_03_r (R)"(ComponentPose=@n27, Alpha=1, ConstraintWeights_0=1) {
    alphaBoolBlend {blendOption=Linear}
    boneToModify {boneName=K_pinky_03_r}
    constraintSetup [{transformType=Rotation, targetBone {boneName=ring_03_r}}]
    constraintWeights [1]
   }
   Output Pose(Result=Component To Local(ComponentPose=@n28)) {layerGroup=None}
  ]}
 EventGraph {nodes [
   n1 = Get Dynamic Material Instance(SkeletalMeshComponent=GetPartSkeletalMeshComponent())
   > n2 = Branch(Condition=Get is_wearing_westsausage() -> (is_wearing_westsausage:bool))
   > n3 = Branch(Condition=Get is_performing_emote() -> (is_performing_emote:bool))
   SetScalarParameterValue(execute=[@n8.then, @n7.then, @n4.then, @n3.then, @n6.then, @n2.else, @n5.then], self=@n1, ParameterName=Hide Element 09, Value=0)
   n4 = Branch(execute=@n3.else, Condition=Get bIsTacticalSprinting())
   n5 = Branch(execute=@n4.else, Condition=Get bIsValetDriver() -> (bIsValetDriver:bool))
   n6 = Branch(execute=@n5.else, Condition=Get bIsMeatballDriver() -> (bIsMeatballDriver:bool))
   n7 = Branch(execute=@n6.else, Condition=Get bIsAntelopeVehicleDriver() -> (bIsAntelopeVehicleDriver:bool))
   n8 = Branch(execute=@n7.else, Condition=Get bIsSkyDiving())
   SetScalarParameterValue(execute=@n8.else, self=@n1, ParameterName=Hide Element 09, Value=1)
   n9 = Cast To FortPlayerPawn(Object=GetOwningActor())
   > Set As Fort Player Pawn(As Fort Player Pawn:Fort Player Pawn Object Reference=@n9) -> (Output_Get:Fort Player Pawn Object Reference)
   > n10 = Get!(execute:Exec) -> (then:Exec, else:Exec, As Fort Player Pawn:Fort Player Pawn Object Reference)
   > Set is_wearing_westsausage(is_wearing_westsausage:bool=HasMatchingGameplayTag!(self:Gameplay Tag Asset Interface Interface=@n10, TagToCheck:Gameplay Tag Structure=(TagName="Granted.Athena.HandsAreReplaced")) -> (ReturnValue:bool)) -> (Output_Get:bool)
   Event BlueprintUpdateAnimation()
   > Sequence!(execute) -> (then_0, then_1, then_2:Exec)
   n11 = Get Upperarm_Mesh() -> (Upperarm_Mesh:Skeletal Mesh Component Object Reference)
   > Set Shoulder_rot_f_r(Shoulder_rot_f_r:double=GetAlphaFromReferenceTransformDelta(SkeletalMeshComponent=@n11, BoneName=upperarm_r, BaseName=clavicle_r, InputRange=(X=0,Y=120), TwistAxis="0,0,1")) -> (Output_Get:double)
   > Set Shoulder_rot_f_l(Shoulder_rot_f_l:double=GetAlphaFromReferenceTransformDelta(SkeletalMeshComponent=@n11, BoneName=upperarm_l, BaseName=clavicle_l, InputRange=(X=0,Y=120), TwistAxis="0,0,1")) -> (Output_Get:double)
   > Set Shoulder_rot_back_l(Shoulder_rot_back_l:double=GetAlphaFromReferenceTransformDelta(SkeletalMeshComponent=@n11, BoneName=upperarm_r, BaseName=clavicle_r, InputRange=(X=0,Y=-90), TwistAxis="0,0,1")) -> (Output_Get:double)
   > Set Shoulder_rot_back_r(Shoulder_rot_back_r:double=GetAlphaFromReferenceTransformDelta(SkeletalMeshComponent=@n11, BoneName=upperarm_l, BaseName=clavicle_l, InputRange=(X=0,Y=-90), TwistAxis="0,0,1")) -> (Output_Get:double)
   n12 = Get Upperarm_Mesh()
   > Set Shoulder_rot_up_r(Shoulder_rot_up_r:double=GetAlphaFromReferenceTransformDelta(SkeletalMeshComponent=@n12, BoneName=upperarm_r, BaseName=clavicle_r, DeltaComponentType=EulerY, InputRange=(X=0,Y=40), TwistAxis="0,0,1")) -> (Output_Get:double)
   > Set Shoulder_rot_up_l(Shoulder_rot_up_l:double=GetAlphaFromReferenceTransformDelta(SkeletalMeshComponent=@n12, BoneName=upperarm_l, BaseName=clavicle_l, DeltaComponentType=EulerY, InputRange=(X=0,Y=40), TwistAxis="0,0,1")) -> (Output_Get:double)
   > Set Shoulder_rot_dn_r(Shoulder_rot_dn_r:double=GetAlphaFromReferenceTransformDelta(SkeletalMeshComponent=@n12, BoneName=upperarm_r, BaseName=clavicle_r, InputRange=(X=-15,Y=-90), TwistAxis="0,0,1")) -> (Output_Get:double)
   > Set Shoulder_rot_dn_l(Shoulder_rot_dn_l:double=GetAlphaFromReferenceTransformDelta(SkeletalMeshComponent=@n12, BoneName=upperarm_l, BaseName=clavicle_l, InputRange=(X=-15,Y=-90), TwistAxis="0,0,1")) -> (Output_Get:double)
   Event BlueprintInitializeAnimation()
   > Set Skel_Mesh(Skel_Mesh=GetOwnerSkeletalMeshComponent())
   > n13 = Cast To PlayerPawn_Athena(Object=TryGetPawnOwner())
   > n14 = Set Pawn Owner(Pawn Owner:Fort Player Pawn Athena Object Reference=@n13) -> (Output_Get:Fort Player Pawn Athena Object Reference)
   > Set Upperarm_Mesh(Upperarm_Mesh:Skeletal Mesh Component Object Reference=GetSkeletalMeshForPartType(self=@n14)) -> (Output_Get:Skeletal Mesh Component Object Reference)
   Event OnBeginPlayingEmote() {is=title K2Node_Event}
   > Set is_performing_emote(is_performing_emote:bool=true) -> (Output_Get:bool)
   Event OnEndPlayingEmote() {is=title K2Node_Event}
   > Set is_performing_emote(is_performing_emote=false)
   Get Skel_Mesh()
   Get Skel_Mesh()
  ]}
]
variables [
 As Fort Player Pawn {type=FortPlayerPawn}
 is_performing_emote {type=boolean}
 is_wearing_westsausage {type=boolean}
 Pawn Owner {type=FortPlayerPawnAthena}
 Shoulder_rot_back_l {type=number}
 Shoulder_rot_back_r {type=number}
 Shoulder_rot_dn_l {type=number}
 Shoulder_rot_dn_r {type=number}
 Shoulder_rot_f_l {type=number}
 Shoulder_rot_f_r {type=number}
 Shoulder_rot_up_l {type=number}
 Shoulder_rot_up_r {type=number}
 Skel_Mesh {type=SkeletalMeshComponent}
 Upperarm_Mesh {type=SkeletalMeshComponent}
]
