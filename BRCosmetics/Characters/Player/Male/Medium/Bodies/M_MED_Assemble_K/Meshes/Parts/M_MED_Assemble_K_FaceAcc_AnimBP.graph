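# Post-copy corrective AnimBP for the face accessory of the male medium Assemble K player body.
# It copies the attached parent's pose, then adjusts neck_01, neck_02 and head_energy with weighted rotation constraints and additive bone-space rotations.
# Each update derives two upward-tilt alphas from head-to-neck_02 and neck_01-to-spine_05 deltas against the reference pose on the owner's skeletal mesh.
# These alphas scale the additive rotations and one neck_02 constraint.
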AnimBlueprint /BRCosmetics/Characters/Player/Male/Medium/Bodies/M_MED_Assemble_K/Meshes/Parts/M_MED_Assemble_K_FaceAcc_AnimBP
target_skeleton=&/M_MED_Assemble_K_FaceAcc_Skeleton
graphs [
 AnimGraph {nodes [
   n1 = "Constraint - neck_02 (RR)"!(
    ComponentPose=Local To Component(LocalPose=Copy Pose From Mesh() {bUseAttachedParent=true})
    Alpha=1
    ConstraintWeights_0=0.4
    ConstraintWeights_1:float=0.6
   ) -> (Pose) {
    alphaBoolBlend {blendOption=Linear}
    boneToModify {boneName=neck_02}
    constraintSetup [{transformType=Rotation, targetBone {boneName=neck_01}}, {transformType=Rotation, targetBone {boneName=neck_02}}]
    constraintWeights [1,1]
   }
   n2 = "Transform (Modify) Bone - Bone: neck_01"(ComponentPose=@n1, Alpha=Get neck01_rot_up_0() -> (neck01_rot_up_0:double), Translation="0,0,0", Rotation="0,25,0", Scale="1,1,1") {
    rotationMode=BMM_Additive
    rotationSpace=BCS_BoneSpace
    scaleMode=BMM_Additive
    scaleSpace=BCS_BoneSpace
    translationMode=BMM_Additive
    translationSpace=BCS_BoneSpace
    alphaBoolBlend {blendOption=Linear}
    boneToModify {boneName=neck_01}
    rotation {yaw=25}
   }
   n3 = Get head_rot_up() -> (head_rot_up:double)
   n4 = "Transform (Modify) Bone - Bone: neck_02"(ComponentPose=@n2, Alpha=@n3, Translation="0,0,0", Rotation="0,10,0", Scale="1,1,1") {
    rotationMode=BMM_Additive
    rotationSpace=BCS_BoneSpace
    scaleMode=BMM_Additive
    scaleSpace=BCS_BoneSpace
    translationMode=BMM_Additive
    translationSpace=BCS_BoneSpace
    alphaBoolBlend {blendOption=Linear}
    boneToModify {boneName=neck_02}
    rotation {yaw=10}
   }
   n5 = "Transform (Modify) Bone - Bone: neck_01"(ComponentPose=@n4, Alpha=@n3, Translation="0,0,0", Rotation="0,20,0", Scale="1,1,1") {
    rotationMode=BMM_Additive
    rotationSpace=BCS_BoneSpace
    scaleMode=BMM_Additive
    scaleSpace=BCS_BoneSpace
    translationMode=BMM_Additive
    translationSpace=BCS_BoneSpace
    alphaBoolBlend {blendOption=Linear}
    boneToModify {boneName=neck_01}
    rotation {yaw=20}
   }
   n6 = "Constraint - neck_02 (RR)"(ComponentPose=@n5, Alpha=@n3, ConstraintWeights_0=0.3, ConstraintWeights_1=0.7) {
    alphaBoolBlend {blendOption=Linear}
    boneToModify {boneName=neck_02}
    constraintSetup [{transformType=Rotation, targetBone {boneName=neck_02}}, {transformType=Rotation, targetBone {boneName=spine_05}}]
    constraintWeights [1,1]
   }
   n7 = "Constraint - head_energy (RR)"!(ComponentPose=@n6, Alpha=1, ConstraintWeights_0=0.5, ConstraintWeights_1:float=0.5) -> (Pose) {
    alphaBoolBlend {blendOption=Linear}
    boneToModify {boneName=head_energy}
    constraintSetup [{transformType=Rotation, targetBone {boneName=neck_01}}, {transformType=Rotation, targetBone {boneName=spine_05}}]
    constraintWeights [1,1]
   }
   n8 = "Transform (Modify) Bone - Bone: head_energy"(ComponentPose=@n7, Alpha=@n3, Translation="0,0,0", Rotation="0,20,0", Scale="1,1,1") {
    rotationMode=BMM_Additive
    rotationSpace=BCS_BoneSpace
    scaleMode=BMM_Additive
    scaleSpace=BCS_BoneSpace
    translationMode=BMM_Additive
    translationSpace=BCS_BoneSpace
    alphaBoolBlend {blendOption=Linear}
    boneToModify {boneName=head_energy}
    rotation {yaw=20}
   }
   Output Pose(Result=Component To Local(ComponentPose=@n8)) {layerGroup=None}
  ]}
 EventGraph {nodes [
   Event BlueprintUpdateAnimation()
   n1 = Get Skel_Mesh()
   > Set head_rot_up(head_rot_up:double=GetAlphaFromReferenceTransformDelta(SkeletalMeshComponent=@n1, BoneName=head, BaseName=neck_02, InputRange=(X=0,Y=-50), TwistAxis="0,0,1")) -> (Output_Get:double)
   > Set neck01_rot_up_0(neck01_rot_up_0:double=GetAlphaFromReferenceTransformDelta(SkeletalMeshComponent=@n1, BoneName=neck_01, BaseName=spine_05, InputRange=(X=0,Y=-50), TwistAxis="0,0,1")) -> (Output_Get:double)
   Event BlueprintInitializeAnimation()
   > Set Skel_Mesh(Skel_Mesh=GetOwnerSkeletalMeshComponent())
  ]}
]
variables [
 head_rot_up {type=number, default=1}
 neck01_rot_up_0 {type=number, default=1}
 Skel_Mesh {type=SkeletalMeshComponent}
]
